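# Control Rig for the Red Sleeves male character's belt, procedurally shifting belt bones as the thighs move.
# A sequence runs spherical pose readers (two right thigh, one left) whose outputs weight pivot and bone offsets, except a constant right-pivot rotation.
# Constraints blend back-left between pelvis and left pivot, and center front/back between sides before the second right-side pass.
# Two alternative center constraints are unconnected, and the first reader draws debug.
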ControlRigBlueprint /BRCosmetics/Characters/Player/Male/Medium/Bodies/M_MED_Red_Sleeves/Meshes/M_MED_Red_Sleeves_belt_CtrlRig
graphs [
 RigVMFunctionLibrary {links [], nodes []}
 RigVMModel {nodes [
   RigUnit_BeginExecution = &RigUnit_BeginExecution()
   n1 = /Script/ControlRig.RigUnit_SequenceExecution(ExecuteContext=@RigUnit_BeginExecution.ExecutePin) {is=struct RigVMUnitNode}
   n2 = &RigUnit_SphericalPoseReader(
    ActiveRegionScaleFactors=(PositiveWidth=0.5,NegativeWidth=4,PositiveHeight=4,NegativeHeight=4)
    ActiveRegionSize=0.1
    Debug=(bDrawDebug=True,DebugScale=25,DebugSegments=20,DebugThickness=0.25)
    DriverAxis=(X=1,Y=0,Z=0)
    DriverItem=(Type=Bone,Name="thigh_r")
    FalloffRegionScaleFactors=(PositiveWidth=1,NegativeWidth=1,PositiveHeight=1,NegativeHeight=1)
    FalloffSize=0.45
    FlipWidthScaling=false
    OptionalParentItem=(Type=Bone,Name="None")
    RotationOffset=(X=-90,Y=0,Z=0)
    ExecutePin=@n1.A
   )
   > &RigUnit_ModifyTransforms(
    ItemToModify=((Item=(Type=Bone,Name="dyn_beltPivot_ft_01_r"),Transform=(Rotation=(X=0,Y=0,Z=0,W=1),Translation=(X=0,Y=0,Z=0),Scale3D=(X=1,Y=1,Z=1))))
    ItemToModify.0.Transform.Rotation=&RigVMFunction_MathQuaternionFromEuler(Euler=(X=0,Y=30,Z=-30)).Result
    Weight=@n2.OutputParam
   )
   > &RigUnit_ModifyTransforms(
    ItemToModify=((Item=(Type=Bone,Name="dyn_beltPivot_ft_02_r"),Transform=(Rotation=(X=0,Y=0,Z=0,W=1),Translation=(X=0,Y=0,Z=0),Scale3D=(X=1,Y=1,Z=1))))
    ItemToModify.0.Transform.Rotation=&RigVMFunction_MathQuaternionFromEuler(Euler=(X=0,Y=20,Z=-20)).Result
    Weight=@n2.OutputParam
   )
   > &RigUnit_ModifyTransforms(
    ItemToModify=((Item=(Type=Bone,Name="dyn_beltPivot_belt_r"),Transform=(Rotation=(X=0,Y=0,Z=0,W=1),Translation=(X=0,Y=0,Z=0),Scale3D=(X=1,Y=1,Z=1))))
    ItemToModify.0.Transform.Rotation=&RigVMFunction_MathQuaternionFromEuler(Euler=(X=0,Y=15,Z=-15)).Result
    Weight=@n2.OutputParam
   )
   > &RigUnit_ModifyTransforms(
    ItemToModify=((Item=(Type=Bone,Name="dyn_beltPivot_bk_r"),Transform=(Rotation=(X=0,Y=0,Z=0,W=1),Translation=(X=0,Y=0,Z=0),Scale3D=(X=1,Y=1,Z=1))))
    ItemToModify.0.Transform.Rotation=&RigVMFunction_MathQuaternionFromEuler(Euler=(X=0,Y=5,Z=-5)).Result
    Weight=@n2.OutputParam
   )
   n3 = &RigUnit_SphericalPoseReader(
    ActiveRegionScaleFactors=(PositiveWidth=4,NegativeWidth=1,PositiveHeight=4,NegativeHeight=4)
    ActiveRegionSize=0.1
    Debug=(bDrawDebug=false,DebugScale=25,DebugSegments=20,DebugThickness=0.25)
    DriverAxis=(X=-1,Y=0,Z=0)
    DriverItem=(Type=Bone,Name="thigh_l")
    FalloffRegionScaleFactors=(PositiveWidth=1,NegativeWidth=1,PositiveHeight=1,NegativeHeight=1)
    FalloffSize=0.45
    FlipWidthScaling=false
    OptionalParentItem=(Type=Bone,Name="pelvis")
    RotationOffset=(X=90,Y=0,Z=0)
    ExecutePin=@n1.B
   )
   > &RigUnit_ModifyTransforms(
    ItemToModify=((Item=(Type=Bone,Name="dyn_beltPivot_l"),Transform=(Rotation=(X=0,Y=0,Z=0,W=1),Translation=(X=3,Y=0,Z=0),Scale3D=(X=1,Y=1,Z=1))))
    ItemToModify.0.Transform.Rotation=&RigVMFunction_MathQuaternionFromEuler(Euler=(X=0,Y=-10,Z=-20)).Result
    Weight=@n3.OutputParam
   )
   > &RigUnit_ModifyTransforms(
    ItemToModify=((Item=(Type=Bone,Name="dyn_belt_l_MSTR"),Transform=(Rotation=(X=0,Y=0,Z=0,W=1),Translation=(X=1,Y=0,Z=0),Scale3D=(X=1,Y=1,Z=1))))
    ItemToModify.0.Transform.Rotation=&RigVMFunction_MathQuaternionFromEuler(Euler=(X=0,Y=0,Z=-15)).Result
    Weight=@n3.OutputParam
   )
   > &RigUnit_ModifyTransforms(
    ItemToModify=((Item=(Type=Bone,Name="dyn_belt_ft_02_l_MSTR"),Transform=(Rotation=(X=0,Y=0,Z=0,W=1),Translation=(X=2,Y=0,Z=0),Scale3D=(X=1,Y=1,Z=1))))
    ItemToModify.0.Transform.Rotation=&RigVMFunction_MathQuaternionFromEuler(Euler=(X=0,Y=0,Z=-30)).Result
    Weight=@n3.OutputParam
   )
   > &RigUnit_ParentConstraint(
    AdvancedSettings=(RotationOrderForFilter=XZY)
    Child=(Type=Bone,Name="dyn_belt_bk_l_MSTR")
    Parents=((Item=(Type=Bone,Name="pelvis"),Weight=0.6),(Item=(Type=Bone,Name="dyn_beltPivot_l"),Weight=0.4))
    Weight=1
   )
   &RigUnit_ParentConstraint(
    AdvancedSettings=(RotationOrderForFilter=XZY)
    Child=(Type=Bone,Name="dyn_belt_ft_MSTR")
    Parents=((Item=(Type=Bone,Name="dyn_belt_ft_01_l_MSTR"),Weight=0.5),(Item=(Type=Bone,Name="dyn_belt_ft_01_r_MSTR"),Weight=0.5))
    Weight=1
    ExecutePin=@n1.C
   )
   > /Script/ControlRig.RigUnit_PositionConstraintLocalSpaceOffset(
    Child=(Type=Bone,Name="dyn_belt_bk_MSTR")
    Filter=(bX=True,bY=True,bZ=True)
    Parents=((Item=(Type=Bone,Name="dyn_belt_bk_l_MSTR"),Weight=0.5),(Item=(Type=Bone,Name="dyn_belt_bk_r_MSTR"),Weight=0.5))
    Weight=1
    bMaintainOffset=True
   ) {is=struct RigVMUnitNode}
   > &RigUnit_ModifyTransforms(
    ItemToModify=((Item=(Type=Bone,Name="dyn_belt_bk_MSTR"),Transform=(Rotation=(X=0,Y=0,Z=0,W=1),Translation=(X=0,Y=0,Z=0),Scale3D=(X=1,Y=1,Z=1))))
    Weight=1
   )
   n4 = &RigUnit_SphericalPoseReader(
    ActiveRegionScaleFactors=(PositiveWidth=0.1,NegativeWidth=3,PositiveHeight=4,NegativeHeight=4)
    ActiveRegionSize=0.1
    Debug=(bDrawDebug=false,DebugScale=25,DebugSegments=20,DebugThickness=0.25)
    DriverAxis=(X=1,Y=0,Z=0)
    DriverItem=(Type=Bone,Name="thigh_r")
    FalloffRegionScaleFactors=(PositiveWidth=1.3,NegativeWidth=1,PositiveHeight=1,NegativeHeight=1.5)
    FalloffSize=0.45
    FlipWidthScaling=false
    OptionalParentItem=(Type=Bone,Name="pelvis")
    RotationOffset=(X=0,Y=20,Z=0)
    ExecutePin=@n1.D
   )
   > &RigUnit_ModifyTransforms(
    ItemToModify=((Item=(Type=Bone,Name="dyn_beltPivot_ft_01_r"),Transform=(Rotation=(X=0,Y=0,Z=0,W=1),Translation=(X=0,Y=0,Z=0),Scale3D=(X=1,Y=1,Z=1))))
    ItemToModify.0.Transform.Rotation=&RigVMFunction_MathQuaternionFromEuler(Euler=(X=0,Y=10,Z=-10)).Result
    Weight=@n4.OutputParam
   )
   > &RigUnit_ModifyTransforms(
    ItemToModify=((Item=(Type=Bone,Name="dyn_beltPivot_ft_02_r"),Transform=(Rotation=(X=0,Y=0,Z=0,W=1),Translation=(X=0,Y=0,Z=0),Scale3D=(X=1,Y=1,Z=1))))
    ItemToModify.0.Transform.Rotation=&RigVMFunction_MathQuaternionFromEuler(Euler=(X=0,Y=10,Z=-10)).Result
    Weight=@n4.OutputParam
   )
   > &RigUnit_ModifyTransforms(
    ItemToModify=((Item=(Type=Bone,Name="dyn_beltPivot_belt_r"),Transform=(Rotation=(X=0,Y=0,Z=0,W=1),Translation=(X=0,Y=0,Z=0),Scale3D=(X=1,Y=1,Z=1))))
    ItemToModify.0.Transform.Rotation=&RigVMFunction_MathQuaternionFromEuler(Euler=(X=0,Y=15,Z=-15)).Result
    Weight=@n4.OutputParam
   )
   > &RigUnit_ModifyTransforms(
    ItemToModify=((Item=(Type=Bone,Name="dyn_beltPivot_bk_r"),Transform=(Rotation=(X=0,Y=0,Z=0,W=1),Translation=(X=0,Y=0,Z=0),Scale3D=(X=1,Y=1,Z=1))))
    ItemToModify.0.Transform.Rotation=&RigVMFunction_MathQuaternionFromEuler(Euler=(X=0,Y=10,Z=-10)).Result
    Weight=@n4.OutputParam
   )
   > &RigUnit_ModifyTransforms(
    ItemToModify=((Item=(Type=Bone,Name="dyn_beltPivot_r"),Transform=(Rotation=(X=0,Y=0,Z=0,W=1),Translation=(X=0,Y=0,Z=0),Scale3D=(X=1,Y=1,Z=1))))
    Weight=1
    ItemToModify.0.Transform.Rotation=&RigVMFunction_MathQuaternionFromEuler(Euler=(X=0,Y=5,Z=-3)).Result
   )
   &RigUnit_ParentConstraint(
    AdvancedSettings=(RotationOrderForFilter=XZY)
    Child=(Type=Bone,Name="dyn_belt_bk_MSTR")
    ExecutePin=()
    Parents=((Item=(Type=Bone,Name="dyn_belt_bk_l_MSTR"),Weight=0.5),(Item=(Type=Bone,Name="dyn_belt_bk_r_MSTR"),Weight=0.5))
    Weight=1
   )
   &RigUnit_PositionConstraintLocalSpaceOffset(
    Child=(Type=Bone,Name="dyn_belt_ft_MSTR")
    ExecutePin=()
    Filter=(bX=True,bY=True,bZ=True)
    Parents=((Item=(Type=Bone,Name="dyn_belt_ft_01_l_MSTR"),Weight=1),(Item=(Type=Bone,Name="dyn_belt_ft_01_r_MSTR"),Weight=1))
    Weight=1
    bMaintainOffset=True
   )
  ]}
]
hierarchy [
 root {loc=<0 0 0>, rot=<0 0 0 1>}
 pelvis {parent=root, loc=<0 -0.7071 93.1191>, rot=<0.0127 0.707 -0.0127 -0.707>}
 spine_01 {parent=pelvis, loc=<6.7494 -0.6388 0>, rot=<0 0 0.0503 -0.9987>}
 spine_02 {parent=spine_01, loc=<7.2493 -0.0308 0>, rot=<0 0 0.0041 -1>}
 spine_03 {parent=spine_02, loc=<6.0084 -0.0167 0>, rot=<0 0 0.0027 -1>}
 spine_04 {parent=spine_03, loc=<8.0999 -0.0224 0>, rot=<0 0 0.0026 -1>}
 spine_05 {parent=spine_04, loc=<10.0119 -0.0184 0>, rot=<0 0 -0.0016 -1>}
 clavicle_l {parent=spine_05, loc=<4.3575 -0.1466 -2.3223>, rot=<0.0476 -0.7289 -0.0102 -0.6829>}
 upperarm_l {parent=clavicle_l, loc=<15.5111 0 0>, rot=<0.0221 -0.3319 -0.0095 -0.943>}
 lowerarm_l {parent=upperarm_l, loc=<26.647 0 0>, rot=<0 0 -0.2447 -0.9696>}
 hand_l {parent=lowerarm_l, loc=<22.344 0 0>, rot=<-0.6794 -0.0546 0.1067 -0.7239>}
 thumb_01_l {parent=hand_l, loc=<1.3118 -0.3191 -2.459>, rot=<0.6546 -0.4042 0.0664 -0.6354>}
 thumb_02_l {parent=thumb_01_l, loc=<5.9336 -0.0448 0.0401>, rot=<-0.0088 0.005 0.2979 -0.9545>}
 thumb_03_l {parent=thumb_02_l, loc=<2.2223 0.3682 0>, rot=<0 0 0.0067 -1>}
 index_metacarpal_l {parent=hand_l, loc=<1.6535 0.0183 -2.96>, rot=<0.0158 -0.0124 -0.0173 -0.9996>}
 index_01_l {parent=index_metacarpal_l, loc=<8.3008 0.0357 -0.0524>, rot=<-0.0044 0.0485 0.0924 -0.9945>}
 index_02_l {parent=index_01_l, loc=<3.3587 -0.3004 0>, rot=<0 0 0.3225 -0.9466>}
 index_03_l {parent=index_02_l, loc=<2.5578 -0.123 0>, rot=<0 0 0.2559 -0.9667>}
 middle_metacarpal_l {parent=hand_l, loc=<2.1856 -0.1725 -0.5984>, rot=<0.006 0.0168 -0.0329 -0.9993>}
 middle_01_l {parent=middle_metacarpal_l, loc=<7.3275 -0.2447 0.0129>, rot=<-0.0208 0.0042 0.0858 -0.9961>}
 middle_02_l {parent=middle_01_l, loc=<4.016 0.1415 0>, rot=<0 0 0.3519 -0.936>}
 middle_03_l {parent=middle_02_l, loc=<2.6103 0.0517 0>, rot=<0 0 0.1707 -0.9853>}
 ring_metacarpal_l {parent=hand_l, loc=<2.2074 -0.3625 1.3638>, rot=<-0.0284 0.0722 -0.0458 -0.9959>}
 ring_01_l {parent=ring_metacarpal_l, loc=<7.0276 -0.6134 -0.2041>, rot=<-0.0484 -0.0225 0.1894 -0.9805>}
 ring_02_l {parent=ring_01_l, loc=<3.6174 0.4647 0>, rot=<0 0 0.2806 -0.9598>}
 ring_03_l {parent=ring_02_l, loc=<2.4152 0.114 0>, rot=<0 0 0.1189 -0.9929>}
 pinky_metacarpal_l {parent=hand_l, loc=<1.7616 -0.9753 3.3814>, rot=<-0.0397 0.0814 -0.069 -0.9935>}
 pinky_01_l {parent=pinky_metacarpal_l, loc=<7.1709 -0.4831 -0.1094>, rot=<-0.0676 -0.0132 0.2236 -0.9723>}
 pinky_02_l {parent=pinky_01_l, loc=<2.8274 0.0893 0>, rot=<0 0 0.2854 -0.9584>}
 pinky_03_l {parent=pinky_02_l, loc=<2.2046 0.0307 0>, rot=<0 0 0.1445 -0.9895>}
 deform_wrist_dn_base_l {parent=hand_l, loc=<0 -2.2342 0>, rot=<0 0 0 -1>}
 deform_wrist_dn_l {parent=deform_wrist_dn_base_l, loc=<0 -0.085 0>, rot=<0 0 0 1>, scale=<0.9929 1 1>}
 deform_wrist_up_base_l {parent=hand_l, loc=<0 2.2682 0>, rot=<0 0 0 -1>}
 deform_wrist_up_l {parent=deform_wrist_up_base_l, loc=<0 0 0>, rot=<0 0 0 1>}
 lowerarm_twist_02_l {parent=lowerarm_l, loc=<9.6199 0 0>, rot=<0 0 0 1>}
 lowerarm_twist_01_l {parent=lowerarm_l, loc=<17.8549 0 0>, rot=<0 0 0 1>}
 deform_elbow_in_base_l {parent=lowerarm_l, loc=<0.2201 5.3074 0.8925>, rot=<0.0388 -0.0107 -0.0004 -0.9992>}
 deform_elbow_in_l {parent=deform_elbow_in_base_l, loc=<0 0 0>, rot=<0 0 0 1>}
 deform_elbow_out_base_l {parent=lowerarm_l, loc=<0.0067 -3.9879 0.7908>, rot=<0.0388 -0.0107 -0.0004 -0.9992>}
 deform_elbow_out_l {parent=deform_elbow_out_base_l, loc=<0 0 0>, rot=<0 0 0 1>}
 upperarm_twist_01_l {parent=upperarm_l, loc=<0.0041 0 0>, rot=<0 0 0 1>}
 deform_upperarm_bk_base_l {parent=upperarm_twist_01_l, loc=<1.1164 -7.2153 -1.9401>, rot=<-0.001 0.009 0 -1>}
 deform_upperarm_bk_l {parent=deform_upperarm_bk_base_l, loc=<0 0 0>, rot=<0 0 0 1>}
 deform_upperarm_fr_base_l {parent=upperarm_twist_01_l, loc=<-0.3194 7.0563 -2.7568>, rot=<-0.001 0.009 0 -1>}
 deform_upperarm_fr_l {parent=deform_upperarm_fr_base_l, loc=<0 0 0>, rot=<0 0 0 1>}
 deform_upperarm_up_base_l {parent=upperarm_twist_01_l, loc=<-3.2811 -1.1459 3.2702>, rot=<-0.001 0.009 0 -1>}
 deform_upperarm_up_l {parent=deform_upperarm_up_base_l, loc=<0 0 0>, rot=<0 0 0 1>}
 upperarm_twist_02_l {parent=upperarm_l, loc=<13.7941 0 0>, rot=<0 0 0 1>}
 neck_01 {parent=spine_05, loc=<11.3138 0.1529 0>, rot=<0 0 -0.2003 -0.9797>}
 neck_02 {parent=neck_01, loc=<4.5627 0 0>, rot=<0 0 0.0504 -0.9987>}
 head {parent=neck_02, loc=<5.3914 0 0>, rot=<0 0 0.115 -0.9934>}
 clavicle_r {parent=spine_05, loc=<4.3575 -0.1466 2.3223>, rot=<0.7289 0.0476 -0.6829 0.0102>}
 upperarm_r {parent=clavicle_r, loc=<-15.5111 0 0>, rot=<0.0221 -0.3319 -0.0095 -0.943>}
 lowerarm_r {parent=upperarm_r, loc=<-26.647 0 0>, rot=<0 0 -0.2447 -0.9696>}
 hand_r {parent=lowerarm_r, loc=<-22.344 0 0>, rot=<-0.6794 -0.0546 0.1067 -0.7239>}
 thumb_01_r {parent=hand_r, loc=<-1.3118 0.3191 2.459>, rot=<0.6546 -0.4042 0.0664 -0.6354>}
 thumb_02_r {parent=thumb_01_r, loc=<-5.9336 0.0448 -0.0401>, rot=<-0.0088 0.005 0.2979 -0.9545>}
 thumb_03_r {parent=thumb_02_r, loc=<-2.2223 -0.3682 0>, rot=<0 0 0.0067 -1>}
 index_metacarpal_r {parent=hand_r, loc=<-1.6535 -0.0183 2.96>, rot=<0.0158 -0.0124 -0.0173 -0.9996>}
 index_01_r {parent=index_metacarpal_r, loc=<-8.3008 -0.0357 0.0524>, rot=<-0.0044 0.0485 0.0924 -0.9945>}
 index_02_r {parent=index_01_r, loc=<-3.3587 0.3004 0>, rot=<0 0 0.3225 -0.9466>}
 index_03_r {parent=index_02_r, loc=<-2.5578 0.123 0>, rot=<0 0 0.2559 -0.9667>}
 middle_metacarpal_r {parent=hand_r, loc=<-2.1856 0.1725 0.5984>, rot=<0.006 0.0168 -0.0329 -0.9993>}
 middle_01_r {parent=middle_metacarpal_r, loc=<-7.3275 0.2447 -0.0129>, rot=<-0.0208 0.0042 0.0858 -0.9961>}
 middle_02_r {parent=middle_01_r, loc=<-4.016 -0.1415 0>, rot=<0 0 0.3519 -0.936>}
 middle_03_r {parent=middle_02_r, loc=<-2.6103 -0.0517 0>, rot=<0 0 0.1707 -0.9853>}
 ring_metacarpal_r {parent=hand_r, loc=<-2.2074 0.3625 -1.3638>, rot=<-0.0284 0.0722 -0.0458 -0.9959>}
 ring_01_r {parent=ring_metacarpal_r, loc=<-7.0276 0.6134 0.2041>, rot=<-0.0484 -0.0225 0.1894 -0.9805>}
 ring_02_r {parent=ring_01_r, loc=<-3.6174 -0.4647 0>, rot=<0 0 0.2806 -0.9598>}
 ring_03_r {parent=ring_02_r, loc=<-2.4152 -0.114 0>, rot=<0 0 0.1189 -0.9929>}
 pinky_metacarpal_r {parent=hand_r, loc=<-1.7616 0.9753 -3.3814>, rot=<-0.0397 0.0814 -0.069 -0.9935>}
 pinky_01_r {parent=pinky_metacarpal_r, loc=<-7.1709 0.4831 0.1094>, rot=<-0.0676 -0.0132 0.2236 -0.9723>}
 pinky_02_r {parent=pinky_01_r, loc=<-2.8274 -0.0893 0>, rot=<0 0 0.2854 -0.9584>}
 pinky_03_r {parent=pinky_02_r, loc=<-2.2046 -0.0307 0>, rot=<0 0 0.1445 -0.9895>}
 deform_wrist_dn_base_r {parent=hand_r, loc=<-0.0004 2.2345 -0.0001>, rot=<0 0 0 -1>}
 deform_wrist_dn_r {parent=deform_wrist_dn_base_r, loc=<0 0.085 0>, rot=<0 0 0 1>, scale=<0.9929 1 1>}
 deform_wrist_up_base_r {parent=hand_r, loc=<0.0003 -2.2684 0.0001>, rot=<0 0 0 -1>}
 deform_wrist_up_r {parent=deform_wrist_up_base_r, loc=<0 0 0>, rot=<0 0 0 1>}
 lowerarm_twist_02_r {parent=lowerarm_r, loc=<-9.6199 0 0>, rot=<0 0 0 -1>}
 lowerarm_twist_01_r {parent=lowerarm_r, loc=<-17.8549 0 0>, rot=<0 0 0 -1>}
 deform_elbow_in_base_r {parent=lowerarm_r, loc=<-0.22 -5.3074 -0.8926>, rot=<0.0388 -0.0107 -0.0004 -0.9992>}
 deform_elbow_in_r {parent=deform_elbow_in_base_r, loc=<-0.0001 0 0.0002>, rot=<0 0 0 1>}
 deform_elbow_out_base_r {parent=lowerarm_r, loc=<-0.0067 3.9879 -0.7908>, rot=<0.0388 -0.0107 -0.0004 -0.9992>}
 deform_elbow_out_r {parent=deform_elbow_out_base_r, loc=<0 0 0>, rot=<0 0 0 1>}
 upperarm_twist_01_r {parent=upperarm_r, loc=<-0.0041 0 0>, rot=<0 0 0 -1>}
 deform_upperarm_bk_base_r {parent=upperarm_twist_01_r, loc=<-1.1166 7.2153 1.9404>, rot=<-0.001 0.009 0 -1>}
 deform_upperarm_bk_r {parent=deform_upperarm_bk_base_r, loc=<0 0 0>, rot=<0 0 0 1>}
 deform_upperarm_fr_base_r {parent=upperarm_twist_01_r, loc=<0.3193 -7.0563 2.7569>, rot=<-0.001 0.009 0 -1>}
 deform_upperarm_fr_r {parent=deform_upperarm_fr_base_r, loc=<0 0 0>, rot=<0 0 0 1>}
 deform_upperarm_up_base_r {parent=upperarm_twist_01_r, loc=<3.2808 1.1459 -3.27>, rot=<-0.001 0.009 0 -1>}
 deform_upperarm_up_r {parent=deform_upperarm_up_base_r, loc=<0 0 0>, rot=<0 0 0 -1>}
 upperarm_twist_02_r {parent=upperarm_r, loc=<-13.7941 0 0>, rot=<0 0 0 -1>}
 deform_pec_aim_base_l {parent=spine_04, loc=<7.8667 12.6198 -2.4123>, rot=<0.0103 -0.4602 0.3978 -0.7937>}
 deform_pec_aim_l {parent=deform_pec_aim_base_l, loc=<0 0 0>, rot=<0 0 0 1>}
 deform_pec_base_l {parent=deform_pec_aim_l, loc=<10.766 0 0>, rot=<0 0 0 1>}
 deform_pec_l {parent=deform_pec_base_l, loc=<0.055 0 0>, rot=<0 0 0 1>}
 deform_pec_aim_base_r {parent=spine_04, loc=<7.8662 12.6197 2.4123>, rot=<-0.4602 -0.0103 0.7937 0.3978>}
 deform_pec_aim_r {parent=deform_pec_aim_base_r, loc=<0 0 0>, rot=<0 0 0 1>}
 deform_pec_base_r {parent=deform_pec_aim_r, loc=<-10.7662 -0.0006 -0.0007>, rot=<0 0 0 1>}
 deform_pec_r {parent=deform_pec_base_r, loc=<0 0 0>, rot=<0 0 0 1>}
 dyn_gun_l_MSTR {parent=spine_03, loc=<3.7466 1.7581 -15.5307>, rot=<-0.0277 0.7066 0.7066 -0.0277>}
 dyn_gun_l {parent=dyn_gun_l_MSTR, loc=<0 0 0>, rot=<0 0 0 1>}
 dyn_gun_r_MSTR {parent=spine_03, loc=<3.7466 1.7581 15.5307>, rot=<0.0277 -0.7066 0.7066 -0.0277>}
 dyn_gun_r {parent=dyn_gun_r_MSTR, loc=<0 0 0>, rot=<0 0 0 1>}
 thigh_l {parent=pelvis, loc=<-2.5693 -0.0109 -7.7501>, rot=<0.0013 0.0462 0.0155 -0.9988>}
 calf_l {parent=thigh_l, loc=<-39.039 0 0>, rot=<0 0 -0.0378 -0.9993>}
 foot_l {parent=calf_l, loc=<-42.5546 0 0>, rot=<0.0427 -0.0505 0.0328 -0.9973>}
 ball_l {parent=foot_l, loc=<-6.1599 13.8264 0.6806>, rot=<0.0001 -0.0001 -0.7063 -0.7079>}
 calf_twist_02_l {parent=calf_l, loc=<-18.8064 -0.0967 0.0397>, rot=<0.0237 0.0095 -0.0035 -0.9997>}
 calf_twist_01_l {parent=calf_l, loc=<-35.813 -0.0967 0.0397>, rot=<0.0237 0.0095 -0.0035 -0.9997>}
 deform_knee_in_base_l {parent=calf_l, loc=<1.311 -5.2925 -0.9166>, rot=<0.0007 -0.0051 -0.0013 -1>}
 deform_knee_in_l {parent=deform_knee_in_base_l, loc=<0 0 0>, rot=<0 0 0 1>}
 deform_knee_out_base_l {parent=calf_l, loc=<-0.0102 4 -0.866>, rot=<0.0007 -0.0051 -0.0013 -1>}
 deform_knee_out_l {parent=deform_knee_out_base_l, loc=<0 0 0>, rot=<0 0 0 1>}
 thigh_twist_01_l {parent=thigh_l, loc=<-21.629 -0.0919 -0.0002>, rot=<0.0013 0 0 -1>}
 dyn_pouch_l_MSTR {parent=thigh_twist_01_l, loc=<10.8388 4.8212 -7.3923>, rot=<0.0408 0.513 0.8571 -0.0218>}
 dyn_pouch_l {parent=dyn_pouch_l_MSTR, loc=<0 0 0>, rot=<0 0 0 1>}
 deform_glute_base_l {parent=thigh_l, loc=<-5.9985 -9 -0.1362>, rot=<0.0049 0.0099 -0.4617 -0.887>}
 deform_glute_l {parent=deform_glute_base_l, loc=<0 0 0>, rot=<0 0 0 1>}
 deform_groin_base_l {parent=thigh_l, loc=<-2.1519 8.1149 -1.7885>, rot=<-0.0002 0.011 0 -0.9999>}
 deform_groin_l {parent=deform_groin_base_l, loc=<0 0 0>, rot=<0 0 0 1>}
 thigh_r {parent=pelvis, loc=<-2.5693 -0.0109 7.7501>, rot=<0.0462 -0.0013 0.9988 0.0155>}
 calf_r {parent=thigh_r, loc=<39.0389 0 0>, rot=<0 0 -0.0378 -0.9993>}
 foot_r {parent=calf_r, loc=<42.5546 0 0>, rot=<0.0427 -0.0505 0.0328 -0.9973>}
 ball_r {parent=foot_r, loc=<6.1599 -13.8264 -0.6806>, rot=<0.0001 -0.0001 -0.7063 -0.7079>}
 calf_twist_02_r {parent=calf_r, loc=<18.8064 0.0967 -0.0397>, rot=<0.0237 0.0095 -0.0035 -0.9997>}
 calf_twist_01_r {parent=calf_r, loc=<35.813 0.0967 -0.0397>, rot=<0.0237 0.0095 -0.0035 -0.9997>}
 deform_knee_in_base_r {parent=calf_r, loc=<-1.311 5.2925 0.9166>, rot=<0.0007 -0.0051 -0.0013 -1>}
 deform_knee_in_r {parent=deform_knee_in_base_r, loc=<0 0 0>, rot=<0 0 0 1>}
 deform_knee_out_base_r {parent=calf_r, loc=<0.0103 -4 0.866>, rot=<0.0007 -0.0051 -0.0013 -1>}
 deform_knee_out_r {parent=deform_knee_out_base_r, loc=<0 0 0>, rot=<0 0 0 1>}
 thigh_twist_01_r {parent=thigh_r, loc=<21.629 0.0919 0.0002>, rot=<0.0013 0 0 -1>}
 dyn_pouch_r_MSTR {parent=thigh_twist_01_r, loc=<-9.5736 -0.5318 8.7141>, rot=<0.7051 -0.0343 -0.031 -0.7076>}
 dyn_pouch_r {parent=dyn_pouch_r_MSTR, loc=<0 0 0>, rot=<0 0 0 1>}
 deform_glute_base_r {parent=thigh_r, loc=<5.9983 9 0.1362>, rot=<0.0049 0.0099 -0.4617 -0.887>}
 deform_glute_r {parent=deform_glute_base_r, loc=<0 0 0>, rot=<0 0 0 1>}
 deform_groin_base_r {parent=thigh_r, loc=<2.1519 -8.1149 1.7885>, rot=<-0.0002 0.011 0 -0.9999>}
 deform_groin_r {parent=deform_groin_base_r, loc=<0 0 0>, rot=<0 0 0 1>}
 dyn_belt_ft_MSTR {parent=pelvis, loc=<-0.2895 10.3344 0>, rot=<0.7197 0.6943 0 0>}
 dyn_belt_ft {parent=dyn_belt_ft_MSTR, loc=<0 0 0>, rot=<0 0 0 1>}
 dyn_belt_bk_MSTR {parent=pelvis, loc=<2.2986 -12.9766 0>, rot=<0 0 0.7197 -0.6943>}
 dyn_belt_bk {parent=dyn_belt_bk_MSTR, loc=<0 0 0>, rot=<0 0 0 1>}
 dyn_beltPivot_l {parent=pelvis, loc=<0 0 0>, rot=<0 0 0 1>}
 dyn_belt_ft_01_l_MSTR {parent=dyn_beltPivot_l, loc=<0.7376 9.9387 -5.3414>, rot=<0.7135 0.6884 -0.0939 0.0906>}
 dyn_belt_ft_01_l {parent=dyn_belt_ft_01_l_MSTR, loc=<0 0 0>, rot=<0 0 0 1>}
 dyn_belt_ft_02_l_MSTR {parent=dyn_beltPivot_l, loc=<2.3541 7.2645 -11.4956>, rot=<0.6649 0.6415 -0.2754 0.2657>}
 dyn_belt_ft_02_l {parent=dyn_belt_ft_02_l_MSTR, loc=<0 0 0>, rot=<0 0 0 1>}
 dyn_belt_l_MSTR {parent=dyn_beltPivot_l, loc=<4.1763 0.5579 -15.1756>, rot=<0.5089 0.491 -0.5089 0.491>}
 dyn_belt_l {parent=dyn_belt_l_MSTR, loc=<0 0 0>, rot=<0 0 0 1>}
 dyn_belt_bk_l_MSTR {parent=dyn_beltPivot_l, loc=<4.4864 -10.6007 -11.1404>, rot=<-0.2754 -0.2657 0.6649 -0.6415>}
 dyn_belt_bk_l {parent=dyn_belt_bk_l_MSTR, loc=<0 0 0>, rot=<0 0 0 1>}
 dyn_beltPivot_r {parent=pelvis, loc=<0 0 0>, rot=<0 0 0 1>}
 dyn_beltPivot_ft_02_r {parent=dyn_beltPivot_r, loc=<0 0 0>, rot=<0 0 0 1>}
 dyn_belt_ft_02_r_MSTR {parent=dyn_beltPivot_ft_02_r, loc=<-2.4379 7.9084 12.3129>, rot=<0.6805 0.6565 0.2343 -0.2261>}
 dyn_belt_ft_02_r {parent=dyn_belt_ft_02_r_MSTR, loc=<0 0 0>, rot=<0 0 0 1>}
 dyn_beltPivot_bk_r {parent=dyn_beltPivot_r, loc=<0 0 0>, rot=<0 0 0 1>}
 dyn_belt_bk_r_MSTR {parent=dyn_beltPivot_bk_r, loc=<-0.0938 -11.2901 11.9934>, rot=<0.2754 0.2657 0.6649 -0.6415>}
 dyn_belt_bk_r {parent=dyn_belt_bk_r_MSTR, loc=<0 0 0>, rot=<0 0 0 1>}
 dyn_beltPivot_belt_r {parent=dyn_beltPivot_r, loc=<0 0 0>, rot=<0 0 0 1>}
 dyn_belt_r_MSTR {parent=dyn_beltPivot_belt_r, loc=<-2.6165 0.8014 17.0818>, rot=<0.5089 0.491 0.5089 -0.491>}
 dyn_belt_r {parent=dyn_belt_r_MSTR, loc=<0 0 0>, rot=<0 0 0 1>}
 dyn_beltPivot_ft_01_r {parent=dyn_beltPivot_r, loc=<0 0 0>, rot=<0 0 0 1>}
 dyn_belt_ft_01_r_MSTR {parent=dyn_beltPivot_ft_01_r, loc=<-1.5715 9.929 5.2881>, rot=<0.7169 0.6917 0.0627 -0.0605>}
 dyn_belt_ft_01_r {parent=dyn_belt_ft_01_r_MSTR, loc=<0 0 0>, rot=<0 0 0 1>}
]
