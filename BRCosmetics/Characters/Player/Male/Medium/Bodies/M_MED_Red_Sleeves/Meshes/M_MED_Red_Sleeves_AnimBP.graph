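# Post-process style AnimBP for the male Red_Sleeves player body, adding secondary motion to gun and belt bones.
# It copies the attached parent's pose, runs deform and belt Control Rigs, constrains gun masters to torso bones, simulates guns with AnimDynamics, then caches.
# An unbound bool blend defaults to right-belt AnimDynamics over that cache; the true branch lowers the guns.
# Pouch constraints and several other belt dynamics nodes are disconnected; EventGraph is empty.
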AnimBlueprint /BRCosmetics/Characters/Player/Male/Medium/Bodies/M_MED_Red_Sleeves/Meshes/M_MED_Red_Sleeves_AnimBP
target_skeleton=&/M_MED_Red_Sleeves_Skeleton
graphs [
 AnimGraph {nodes [
   n1 = Anim Dynamics - Bone: dyn_belt_bk_l(Alpha=1, ExternalForce="0,0,0") {
    angularDampingOverride=1
    angularSpringConstant=1500
    bAngularSpring=true
    bOverrideAngularDamping=true
    bUseGravityOverride=true
    simulationSpace=BoneRelative
    alphaBoolBlend {blendOption=Linear}
    boundBone {boneName=dyn_belt_bk_l}
    componentAppliedLinearAccClamp=<1000 1000 1000>
    componentLinearAccScale=<0.45 0.45 0.45>
    componentLinearVelScale=<0.2 0.2 0.2>
    gravityOverride {z=-980}
    physicsBodyDefinitions [{
      boundBone {boneName=dyn_belt_bk_l}
      boxExtents=<25 15 25>
      constraintSetup {angularLimitsMax=<0 5 5>, angularLimitsMin=<0 -5 -5>, angularTarget=<1 0 0>, linearAxesMax=<0 2 0>}
      localJointOffset=<0 -4 0>
     }]
    relativeSpaceBone {boneName=pelvis}
   }
   Anim Dynamics - Bone: dyn_belt_bk_r(ComponentPose=@n1, Alpha=1, ExternalForce="0,0,0") {
    angularDampingOverride=1
    angularSpringConstant=20000
    bAngularSpring=true
    bOverrideAngularDamping=true
    bUseGravityOverride=true
    simulationSpace=BoneRelative
    alphaBoolBlend {blendOption=Linear}
    alphaScaleBiasClamp {scale=0.5}
    boundBone {boneName=dyn_belt_bk_r}
    componentAppliedLinearAccClamp=<1000 1000 1000>
    componentLinearAccScale=<0.45 0.45 0.45>
    componentLinearVelScale=<0.2 0.2 0.2>
    gravityOverride {z=-980}
    physicsBodyDefinitions [{
      boundBone {boneName=dyn_belt_bk_r}
      boxExtents=<20 10 20>
      constraintSetup {angularLimitsMax=<0 5 5>, angularLimitsMin=<0 -5 -5>, angularTarget=<1 0 0>, linearAxesMax=<0 2 0>}
      localJointOffset=<0 -4 0>
     }]
    relativeSpaceBone {boneName=pelvis}
   }
   Anim Dynamics - Bone: dyn_beltPivot_r(Alpha=1, ExternalForce="0,0,0") {
    angularDampingOverride=1
    angularSpringConstant=1500
    bAngularSpring=true
    bOverrideAngularDamping=true
    bUseGravityOverride=true
    simulationSpace=BoneRelative
    alphaBoolBlend {blendOption=Linear}
    boundBone {boneName=dyn_beltPivot_r}
    componentAppliedLinearAccClamp=<1000 1000 1000>
    componentLinearAccScale=<0.45 0.45 0.45>
    componentLinearVelScale=<0.2 0.2 0.2>
    gravityOverride {z=-980}
    physicsBodyDefinitions [{
      boundBone {boneName=dyn_beltPivot_r}
      boxExtents=<10 30 30>
      constraintSetup {angularLimitsMax=<0 5 5>, angularLimitsMin=<0 -5 -5>, angularTarget=<1 0 0>}
      localJointOffset=<-5 0 0>
     }]
    relativeSpaceBone {boneName=pelvis}
   }
   Anim Dynamics - Bone: dyn_beltPivot_r(Alpha=1, ExternalForce="0,0,0") {
    angularDampingOverride=1
    angularSpringConstant=1500
    bOverrideAngularDamping=true
    bUseGravityOverride=true
    simulationSpace=BoneRelative
    alphaBoolBlend {blendOption=Linear}
    boundBone {boneName=dyn_beltPivot_r}
    componentAppliedLinearAccClamp=<3000 3000 3000>
    gravityOverride {z=-980}
    physicsBodyDefinitions [{
      boundBone {boneName=dyn_beltPivot_r}
      boxExtents=<20 50 50>
      constraintSetup {angularLimitsMax=<0 0 0>, angularLimitsMin=<0 0 0>, angularTarget=<1 0 0>}
      localJointOffset=<-12 0 0>
     }]
    relativeSpaceBone {boneName=root}
   }
   Anim Dynamics - Bone: dyn_beltPivot_r(Alpha=1, ExternalForce="0,0,0") {
    angularDampingOverride=1
    angularSpringConstant=1500
    bAngularSpring=true
    bOverrideAngularDamping=true
    bUseGravityOverride=true
    simulationSpace=BoneRelative
    alphaBoolBlend {blendOption=Linear}
    boundBone {boneName=dyn_beltPivot_r}
    componentAppliedLinearAccClamp=<1000 1000 1000>
    gravityOverride {z=-980}
    physicsBodyDefinitions [{
      boundBone {boneName=dyn_beltPivot_r}
      boxExtents=<25 50 50>
      constraintSetup {angularLimitsMax=<0 8 2>, angularLimitsMin=<0 0 -2>, angularTarget=<1 0 0>}
      localJointOffset=<-10 0 6>
     }]
    relativeSpaceBone {boneName=pelvis}
   }
   Anim Dynamics - Bone: dyn_beltPivot_r(Alpha=1, ExternalForce="0,0,0") {
    angularDampingOverride=1
    angularSpringConstant=1500
    bAngularSpring=true
    bOverrideAngularDamping=true
    bUseGravityOverride=true
    simulationSpace=BoneRelative
    alphaBoolBlend {blendOption=Linear}
    boundBone {boneName=dyn_beltPivot_r}
    componentAppliedLinearAccClamp=<1000 1000 1000>
    componentLinearAccScale=<0.45 0.45 0.45>
    componentLinearVelScale=<0.2 0.2 0.2>
    gravityOverride {z=-980}
    physicsBodyDefinitions [{
      boundBone {boneName=dyn_beltPivot_r}
      boxExtents=<15 15 25>
      constraintSetup {angularLimitsMax=<0 5 5>, angularLimitsMin=<0 -5 -5>, angularTarget=<1 0 0>}
      localJointOffset=<-8 0 0>
     }]
    relativeSpaceBone {boneName=pelvis}
   }
   n2 = "Constraint - dyn_pouch_l_MSTR (PPPP)"!(
    ComponentPose
    Alpha=1
    ConstraintWeights_0=0.11
    ConstraintWeights_1:float=0.05
    ConstraintWeights_2:float=0.73
    ConstraintWeights_3:float=0.11
   ) -> (Pose) {
    alphaBoolBlend {blendOption=Linear}
    boneToModify {boneName=dyn_pouch_l_MSTR}
    constraintSetup [
     {targetBone {boneName=pelvis}}
     {targetBone {boneName=thigh_l}}
     {targetBone {boneName=thigh_twist_01_l}}
     {targetBone {boneName=deform_groin_l}}
    ]
    constraintWeights [1,1,1,1]
   }
   "Constraint - dyn_pouch_r_MSTR (PPPP)"!(
    ComponentPose=@n2
    Alpha=1
    ConstraintWeights_0=0.09
    ConstraintWeights_1:float=0.06
    ConstraintWeights_2:float=0.8
    ConstraintWeights_3:float=0.05
   ) -> (Pose) {
    alphaBoolBlend {blendOption=Linear}
    boneToModify {boneName=dyn_pouch_r_MSTR}
    constraintSetup [
     {targetBone {boneName=pelvis}}
     {targetBone {boneName=thigh_r}}
     {targetBone {boneName=thigh_twist_01_r}}
     {targetBone {boneName=deform_groin_r}}
    ]
    constraintWeights [1,1,1,1]
   }
   n3 = Anim Dynamics - Bone: dyn_belt_ft_02_r(ComponentPose=Local To Component(LocalPose=Use cached pose 'FEcache'()), Alpha=1, ExternalForce="0,0,0") {
    angularDampingOverride=1
    angularSpringConstant=1500
    bAngularSpring=true
    bOverrideAngularDamping=true
    bUseGravityOverride=true
    alphaBoolBlend {blendOption=Linear}
    boundBone {boneName=dyn_belt_ft_02_r}
    componentAppliedLinearAccClamp=<1000 1000 1000>
    componentLinearAccScale=<0.45 0.45 0.55>
    componentLinearVelScale=<0.2 0.2 0.2>
    gravityOverride {z=-980}
    physicsBodyDefinitions [{
      boundBone {boneName=dyn_belt_ft_02_r}
      boxExtents=<15 5 15>
      constraintSetup {angularLimitsMax=<0 5 5>, angularLimitsMin=<0 -5 -5>, angularTarget=<1 0 0>}
      localJointOffset=<0 -4 0>
     }]
    relativeSpaceBone {boneName=pelvis}
   }
   n4 = Anim Dynamics - Bone: dyn_belt_r(ComponentPose=@n3, Alpha=1, ExternalForce="0,0,0") {
    angularDampingOverride=1
    angularSpringConstant=1500
    bAngularSpring=true
    bOverrideAngularDamping=true
    bUseGravityOverride=true
    simulationSpace=BoneRelative
    alphaBoolBlend {blendOption=Linear}
    boundBone {boneName=dyn_belt_r}
    componentAppliedLinearAccClamp=<1000 1000 1000>
    componentLinearAccScale=<0.45 0.45 0.55>
    componentLinearVelScale=<0.2 0.2 0.2>
    gravityOverride {z=-980}
    physicsBodyDefinitions [{
      boundBone {boneName=dyn_belt_r}
      boxExtents=<10 5 10>
      constraintSetup {angularLimitsMax=<0 5 5>, angularLimitsMin=<0 -5 -5>, angularTarget=<1 0 0>}
      localJointOffset=<0 -4 0>
     }]
    relativeSpaceBone {boneName=pelvis}
   }
   n5 = Anim Dynamics - Bone: dyn_belt_ft_01_r(ComponentPose=@n4, Alpha=1, ExternalForce="0,0,0") {
    angularDampingOverride=1
    angularSpringConstant=1500
    bAngularSpring=true
    bOverrideAngularDamping=true
    bUseGravityOverride=true
    simulationSpace=BoneRelative
    alphaBoolBlend {blendOption=Linear}
    boundBone {boneName=dyn_belt_ft_01_r}
    componentAppliedLinearAccClamp=<1000 1000 1000>
    componentLinearAccScale=<0.45 0.45 0.55>
    componentLinearVelScale=<0.2 0.2 0.2>
    gravityOverride {z=-980}
    physicsBodyDefinitions [{
      boundBone {boneName=dyn_belt_ft_01_r}
      boxExtents=<10 5 10>
      constraintSetup {angularLimitsMax=<0 5 5>, angularLimitsMin=<0 -5 -5>, angularTarget=<1 0 0>}
      localJointOffset=<0 -4 0>
     }]
    relativeSpaceBone {boneName=pelvis}
   }
   n6 = Anim Dynamics - Bone: dyn_belt_bk_r(ComponentPose=@n5, Alpha=1, ExternalForce="0,0,0") {
    angularDampingOverride=1
    angularSpringConstant=1500
    bAngularSpring=true
    bOverrideAngularDamping=true
    bUseGravityOverride=true
    simulationSpace=BoneRelative
    alphaBoolBlend {blendOption=Linear}
    boundBone {boneName=dyn_belt_bk_r}
    componentAppliedLinearAccClamp=<1000 1000 1000>
    componentLinearAccScale=<0.45 0.45 0.55>
    componentLinearVelScale=<0.2 0.2 0.2>
    gravityOverride {z=-980}
    physicsBodyDefinitions [{
      boundBone {boneName=dyn_belt_bk_r}
      boxExtents=<10 5 10>
      constraintSetup {angularLimitsMax=<0 5 5>, angularLimitsMin=<0 -5 -5>, angularTarget=<1 0 0>}
      localJointOffset=<0 -4 0>
     }]
    relativeSpaceBone {boneName=pelvis}
   }
   n7 = "Transform (Modify) Bone - Bone: dyn_gun_l_MSTR"(ComponentPose=Local To Component(LocalPose=Use cached pose 'FEcache'()), Alpha=1, Translation="0,0,-1.3", Rotation="0,0,0", Scale="1,1,1") {translationMode=BMM_Additive, alphaBoolBlend {blendOption=Linear}, boneToModify {boneName=dyn_gun_l_MSTR}, translation {z=-1.3}}
   n8 = "Transform (Modify) Bone - Bone: dyn_gun_r_MSTR"(ComponentPose=@n7, Alpha=1, Translation="0,0,-1.3", Rotation="0,0,0", Scale="1,1,1") {translationMode=BMM_Additive, alphaBoolBlend {blendOption=Linear}, boneToModify {boneName=dyn_gun_r_MSTR}, translation {z=-1.3}}
   Output Pose(Result=Blend Poses by bool(BlendPose_0=Component To Local(ComponentPose=@n8), BlendPose_1=Component To Local(ComponentPose=@n6)) {blendType=Linear}) {layerGroup=None}
   n9 = Control Rig(Source=Copy Pose From Mesh() {bUseAttachedParent=true}) {
    alphaBoolBlend {blendOption=Linear}
    controlRigAssetReference {blueprintRigClass=&/M_MED_Red_Sleeves_DfrmRig_CtrlRig.M_MED_Red_Sleeves_DfrmRig_CtrlRig_C}
   }
   n10 = Control Rig(Source=@n9) {
    alphaBoolBlend {blendOption=Linear}
    controlRigAssetReference {blueprintRigClass=&/M_MED_Red_Sleeves_belt_CtrlRig.M_MED_Red_Sleeves_belt_CtrlRig_C}
   }
   n11 = "Constraint - dyn_gun_l_MSTR (PPPPPPPP)"!(
    ComponentPose=Local To Component(LocalPose=@n10)
    Alpha=1
    ConstraintWeights_0=0.05
    ConstraintWeights_1:float=0.11
    ConstraintWeights_2:float=0.24
    ConstraintWeights_3:float=0.3
    ConstraintWeights_4:float=0.13
    ConstraintWeights_5:float=0.1
    ConstraintWeights_6:float=0.04
    ConstraintWeights_7:float=0.03
   ) -> (Pose) {
    alphaBoolBlend {blendOption=Linear}
    boneToModify {boneName=dyn_gun_l_MSTR}
    constraintSetup [
     {targetBone {boneName=pelvis}}
     {targetBone {boneName=spine_01}}
     {targetBone {boneName=spine_02}}
     {targetBone {boneName=spine_03}}
     {targetBone {boneName=spine_04}}
     {targetBone {boneName=clavicle_l}}
     {targetBone {boneName=upperarm_twist_01_l}}
     {targetBone {boneName=deform_pec_l}}
    ]
    constraintWeights [1,1,1,1,1,1,1,1]
   }
   n12 = "Constraint - dyn_gun_r_MSTR (PPPPPPPP)"!(
    ComponentPose=@n11
    Alpha=1
    ConstraintWeights_0=0.05
    ConstraintWeights_1:float=0.11
    ConstraintWeights_2:float=0.24
    ConstraintWeights_3:float=0.3
    ConstraintWeights_4:float=0.13
    ConstraintWeights_5:float=0.1
    ConstraintWeights_6:float=0.04
    ConstraintWeights_7:float=0.03
   ) -> (Pose) {
    alphaBoolBlend {blendOption=Linear}
    boneToModify {boneName=dyn_gun_r_MSTR}
    constraintSetup [
     {targetBone {boneName=pelvis}}
     {targetBone {boneName=spine_01}}
     {targetBone {boneName=spine_02}}
     {targetBone {boneName=spine_03}}
     {targetBone {boneName=spine_04}}
     {targetBone {boneName=clavicle_r}}
     {targetBone {boneName=upperarm_twist_01_r}}
     {targetBone {boneName=deform_pec_r}}
    ]
    constraintWeights [1,1,1,1,1,1,1,1]
   }
   n13 = Anim Dynamics - Bone: dyn_gun_l(ComponentPose=@n12, Alpha=1, ExternalForce="0,0,0") {
    angularDampingOverride=1
    angularSpringConstant=1500
    bAngularSpring=true
    bOverrideAngularDamping=true
    bUseGravityOverride=true
    simulationSpace=RootRelative
    alphaBoolBlend {blendOption=Linear}
    alphaScaleBiasClamp {scale=0.25}
    boundBone {boneName=dyn_gun_l}
    componentAppliedLinearAccClamp=<3000 3000 3000>
    componentLinearAccScale=<0.3 0.3 0.55>
    componentLinearVelScale=<0.15 0.15 0.25>
    gravityOverride {z=-980}
    physicsBodyDefinitions [{
      boundBone {boneName=dyn_gun_l}
      boxExtents=<18 25 25>
      constraintSetup {angularLimitsMax=<15 15 15>, angularLimitsMin=<-15 -15 -15>, angularTarget=<1 0 0>}
      localJointOffset=<4 0 0>
     }]
   }
   n14 = Anim Dynamics - Bone: dyn_gun_r(ComponentPose=@n13, Alpha=1, ExternalForce="0,0,0") {
    angularDampingOverride=1
    angularSpringConstant=1500
    bAngularSpring=true
    bOverrideAngularDamping=true
    bUseGravityOverride=true
    simulationSpace=BoneRelative
    alphaBoolBlend {blendOption=Linear}
    alphaScaleBiasClamp {scale=0.25}
    boundBone {boneName=dyn_gun_r}
    componentAppliedLinearAccClamp=<3000 3000 3000>
    componentLinearAccScale=<0.3 0.3 0.55>
    componentLinearVelScale=<0.15 0.15 0.25>
    gravityOverride {z=-980}
    physicsBodyDefinitions [{
      boundBone {boneName=dyn_gun_r}
      boxExtents=<18 25 25>
      constraintSetup {angularLimitsMax=<15 15 15>, angularLimitsMin=<-15 -15 -15>, angularTarget=<1 0 0>}
      localJointOffset=<4 0 0>
     }]
    relativeSpaceBone {boneName=spine_05}
   }
   Save cached pose 'FEcache'(Pose=Component To Local(ComponentPose=@n14)) {cacheName=FEcache}
  ]}
 EventGraph {nodes [Event BlueprintUpdateAnimation()]}
]
variables []
