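# Physics asset for the face accessory mesh of the male medium Red Sleeves player cosmetic, set up for dynamic hair motion.
# Small capsules cover thirteen dyn_hair bones across front, back, left and right, with one larger capsule on the head.
# Each hair bone is constrained directly to the head rather than chained, with narrow swing limits that vary slightly among front strands and an almost locked twist.
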PhysicsAsset /BRCosmetics/Characters/Player/Male/Medium/Bodies/M_MED_Red_Sleeves/Meshes/Parts/M_MED_Red_Sleeves_FaceAcc_Physics
bodies {
 dyn_hair_00_ft [{shapeType=Capsule, length=5, radius=1.75, shapeName=dyn_hair_01_ft_capsule, center=<2.9295 0 0>, rotation=<rot 90 0 0>}]
 dyn_hair_01_bk [{shapeType=Capsule, length=3, radius=2, shapeName=dyn_hair_01_ft_capsule, center=<1 0 0>, rotation=<rot 90 0 0>}]
 dyn_hair_01_ft [{shapeType=Capsule, length=3, radius=2, shapeName=dyn_hair_01_ft_capsule, center=<0.8611 0 0>, rotation=<rot 90 0 0>}]
 dyn_hair_01_l [{shapeType=Capsule, length=3.5, radius=1.5, shapeName=dyn_hair_01_ft_capsule, center=<2 0 0>, rotation=<rot 90 0 0>}]
 dyn_hair_01_r [{shapeType=Capsule, length=3, radius=2, shapeName=dyn_hair_01_ft_capsule, center=<1 0 0>, rotation=<rot 90 0 0>}]
 dyn_hair_02_bk [{shapeType=Capsule, length=3, radius=2, shapeName=dyn_hair_01_ft_capsule, center=<1 0 0>, rotation=<rot 90 0 0>}]
 dyn_hair_02_ft [{shapeType=Capsule, length=3, radius=2, shapeName=dyn_hair_01_ft_capsule, center=<1.093 0 0>, rotation=<rot 90 0 0>}]
 dyn_hair_02_l [{shapeType=Capsule, length=3, radius=2, shapeName=dyn_hair_01_ft_capsule, center=<1 0 0>, rotation=<rot 90 0 0>}]
 dyn_hair_02_r [{shapeType=Capsule, length=3, radius=2, shapeName=dyn_hair_01_ft_capsule, center=<1 0 0>, rotation=<rot 90 0 0>}]
 dyn_hair_03_bk [{shapeType=Capsule, length=3, radius=2, shapeName=dyn_hair_01_ft_capsule, center=<1 0 0>, rotation=<rot 90 0 0>}]
 dyn_hair_03_ft [{shapeType=Capsule, length=3, radius=2, shapeName=dyn_hair_01_ft_capsule, center=<2 0 0>, rotation=<rot 90 0 0>}]
 dyn_hair_03_l [{shapeType=Capsule, length=3, radius=2, shapeName=dyn_hair_01_ft_capsule, center=<1 0 0>, rotation=<rot 90 0 0>}]
 dyn_hair_03_r [{shapeType=Capsule, length=3, radius=2, shapeName=dyn_hair_01_ft_capsule, center=<1 0 0>, rotation=<rot 90 0 0>}]
 head [{shapeType=Capsule, length=11.0756, radius=4.2303, center=<3.8346 4.7263 -0.6216>, rotation=<rot 0 90 90>}]
}
constraints [
 {bone1Name=dyn_hair_00_ft, bone2Name=head, swing1LimitDegrees=10, swing2LimitDegrees=10, twistLimitDegrees=1}
 {bone1Name=dyn_hair_03_l, bone2Name=head, swing1LimitDegrees=15, swing2LimitDegrees=15, twistLimitDegrees=1}
 {bone1Name=dyn_hair_02_l, bone2Name=head, swing1LimitDegrees=15, swing2LimitDegrees=15, twistLimitDegrees=1}
 {bone1Name=dyn_hair_01_l, bone2Name=head, swing1LimitDegrees=15, swing2LimitDegrees=15, twistLimitDegrees=1}
 {bone1Name=dyn_hair_01_r, bone2Name=head, swing1LimitDegrees=15, swing2LimitDegrees=15, twistLimitDegrees=1}
 {bone1Name=dyn_hair_02_r, bone2Name=head, swing1LimitDegrees=15, swing2LimitDegrees=15, twistLimitDegrees=1}
 {bone1Name=dyn_hair_03_r, bone2Name=head, swing1LimitDegrees=15, swing2LimitDegrees=15, twistLimitDegrees=1}
 {bone1Name=dyn_hair_03_bk, bone2Name=head, swing1LimitDegrees=15, swing2LimitDegrees=15, twistLimitDegrees=1}
 {bone1Name=dyn_hair_02_bk, bone2Name=head, swing1LimitDegrees=15, swing2LimitDegrees=15, twistLimitDegrees=1}
 {bone1Name=dyn_hair_01_bk, bone2Name=head, swing1LimitDegrees=15, swing2LimitDegrees=15, twistLimitDegrees=1}
 {bone1Name=dyn_hair_03_ft, bone2Name=head, swing1LimitDegrees=15, swing2LimitDegrees=15, twistLimitDegrees=1}
 {bone1Name=dyn_hair_02_ft, bone2Name=head, swing1LimitDegrees=20, swing2LimitDegrees=20, twistLimitDegrees=1}
 {bone1Name=dyn_hair_01_ft, bone2Name=head, swing1LimitDegrees=20, swing2LimitDegrees=20, twistLimitDegrees=1}
]
